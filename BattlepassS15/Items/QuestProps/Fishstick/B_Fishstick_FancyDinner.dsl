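; Blueprint for a Fishstick "Fancy Dinner" quest prop from the Battle Pass S15 items, which animates the character's head.
; On BeginPlay it calls the parent, then plays the alien fish head facial idle animation, looping, directly on the M_MED_AUX_Alien_Head_01 component.
; The overlap and tick events are disabled stubs, and the construction script only calls its parent, so no quest logic appears in this graph.
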(event EventBeginPlay
  (|CallParentFunction)
  (Components|Animation|PlayAnimation (Variables|Default|GetM_MED_AUX_Alien_Head_01) "/BRCosmetics/Animation/Game/MainPlayer/Facial/Alien_Fish_Head/alien_idle.alien_idle" true))

(event Collision|EventActorBeginOverlap
  (params (OtherActor :type Actor%20Object%20Reference))
  (|CallParentFunction OtherActor) ; This node is disabled and will not be called.\nDrag off pins to build functionality.
)

(event EventTick
  (params (DeltaSeconds :type Float%20%28single-precision%29))
  (|CallParentFunction DeltaSeconds) ; This node is disabled and will not be called.\nDrag off pins to build functionality.
)

(fn ConstructionScript
  (|CallParentFunction))
